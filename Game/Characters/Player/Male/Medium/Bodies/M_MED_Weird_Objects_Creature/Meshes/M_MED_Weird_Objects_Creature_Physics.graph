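# Physics asset for the Weird Objects Creature medium male player body, covering two dynamic vine bones on the left arm.
# It has four single-capsule bodies: the lower-arm and upper-arm vines, plus a twist bone on each of those arm segments.
# Each vine is constrained to the twist bone of its own segment with matching swing and twist angle limits, forming two independent pairs.
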PhysicsAsset /Game/Characters/Player/Male/Medium/Bodies/M_MED_Weird_Objects_Creature/Meshes/M_MED_Weird_Objects_Creature_Physics
bodies {
 dyn_lowerarm_vine_01_l [{shapeType=Capsule, length=3.7637, radius=1.8922, center=<-0.217 0 -0.0001>, rotation=<rot 0 -90 90>}]
 dyn_upperarm_vine_01_l [{shapeType=Capsule, length=2.249, radius=1.4705, center=<-0.2239 0 0>, rotation=<rot 0 -90 90>}]
 lowerarm_twist_02_l [{shapeType=Capsule, length=0, radius=1.1907, center=<-1.869 -0.1676 0.8159>, rotation=<rot 0 80.8499 -78.0217>}]
 upperarm_twist_01_l [{shapeType=Capsule, length=1.8718, radius=4.798, center=<6.7873 -0.6806 0.2145>, rotation=<rot 0 86.9671 -88.0597>}]
}
constraints [
 {bone1Name=dyn_lowerarm_vine_01_l, bone2Name=lowerarm_twist_02_l, swing1LimitDegrees=45, swing2LimitDegrees=45, twistLimitDegrees=45}
 {bone1Name=dyn_upperarm_vine_01_l, bone2Name=upperarm_twist_01_l, swing1LimitDegrees=45, swing2LimitDegrees=45, twistLimitDegrees=45}
]
