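# Post-process style AnimBlueprint for the Weird Objects Creature player body, layering secondary motion and bone corrections onto the pose copied from its parent mesh.
# Control Rig feeds root-space RigidBody physics, then a dyn_backpack offset enabled when the pawn mesh's FrontEnd_Idle_Bypass curve is positive.
# Look At nodes aim wolf thighs at the calf bones, followed by additive thigh and calf adjustments.
# A jump-boots tag flag is updated but unused here.
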AnimBlueprint /Game/Characters/Player/Male/Medium/Bodies/M_MED_Weird_Objects_Creature/Meshes/M_MED_Weird_Objects_Creature_AnimBP
target_skeleton=&/M_MED_Weird_Objects_Creature_Skeleton
graphs [
 AnimGraph {nodes [
   n1 = Control Rig!(Source=Copy Pose From Mesh() {bUseAttachedParent=true}, Alpha, MyFloat:double=0) -> (Pose) {
    alphaBoolBlend {blendOption=Linear}
    controlRigAssetReference {blueprintRigClass=&/M_MED_Weird_Objects_Creature_ControlRig.M_MED_Weird_Objects_Creature_ControlRig_C}
   }
   n2 = RigidBody(ComponentPose=Local To Component(LocalPose=@n1)) {
    bClampLinearTranslationLimitToRefPose=true
    bOverrideWorldGravity=true
    overridePhysicsAsset=&/M_MED_Weird_Objects_Creature_Physics
    simulationSpace=BaseBoneSpace
    alphaBoolBlend {blendOption=Linear}
    baseBoneRef {boneName=root}
    componentAppliedLinearAccClamp=<1100 1100 1200>
    componentLinearAccScale=<1 1 0.7>
    componentLinearVelScale=<1 1 0.7>
    overrideWorldGravity {z=-980}
   }
   n3 = "Transform (Modify) Bone - Bone: dyn_backpack"(ComponentPose=@n2, Alpha=Get FrontEnd_Idle_Float() -> (FrontEnd_Idle_Float:double), Translation="0,0,0", Rotation="0,-20,-20", Scale="1,1,1") {
    rotationMode=BMM_Additive
    rotationSpace=BCS_BoneSpace
    translationMode=BMM_Additive
    translationSpace=BCS_BoneSpace
    alphaBoolBlend {blendOption=Linear}
    boneToModify {boneName=dyn_backpack}
    rotation {roll=-20, yaw=-20}
   }
   n4 = Look At - Bone: wolf_thigh_l(ComponentPose=@n3, Alpha=1) {
    alphaBoolBlend {blendOption=Linear}
    boneToModify {boneName=wolf_thigh_l}
    lookAtLocation {x=0}
    lookAtTarget {boneReference {boneName=calf_l}}
    lookAt_Axis {axis=<-1 0>}
   }
   n5 = Look At - Bone: wolf_thigh_r(ComponentPose=@n4, Alpha=1) {
    alphaBoolBlend {blendOption=Linear}
    boneToModify {boneName=wolf_thigh_r}
    lookAtLocation {x=0}
    lookAtTarget {boneReference {boneName=calf_r}}
    lookAt_Axis {axis=<1 0>}
   }
   n6 = "Transform (Modify) Bone - Bone: wolf_thigh_l"(ComponentPose=@n5, Alpha=1, Translation="0,0,0", Rotation="3,10,0", Scale="1,1,1") {
    rotationMode=BMM_Additive
    rotationSpace=BCS_BoneSpace
    alphaBoolBlend {blendOption=Linear}
    boneToModify {boneName=wolf_thigh_l}
    rotation {pitch=3, yaw=10}
   }
   n7 = "Transform (Modify) Bone - Bone: wolf_thigh_r"(ComponentPose=@n6, Alpha=1, Translation="0,0,0", Rotation="3,10,0", Scale="1,1,1") {
    rotationMode=BMM_Additive
    rotationSpace=BCS_BoneSpace
    alphaBoolBlend {blendOption=Linear}
    boneToModify {boneName=wolf_thigh_r}
    rotation {pitch=3, yaw=10}
   }
   n8 = "Transform (Modify) Bone - Bone: wolf_calf_l"(ComponentPose=@n7, Alpha=1, Translation="0,0,0", Rotation="-5,-10,0", Scale="0.2,0.2,0.2") {
    rotationMode=BMM_Additive
    rotationSpace=BCS_BoneSpace
    scaleMode=BMM_Additive
    scaleSpace=BCS_BoneSpace
    alphaBoolBlend {blendOption=Linear}
    boneToModify {boneName=wolf_calf_l}
    rotation {pitch=-5, yaw=-10}
    scale=<0.2 0.2 0.2>
   }
   n9 = "Transform (Modify) Bone - Bone: wolf_calf_r"(ComponentPose=@n8, Alpha=1, Translation="0,0,0", Rotation="-5,-10,0", Scale="0.2,0.2,0.2") {
    rotationMode=BMM_Additive
    rotationSpace=BCS_BoneSpace
    scaleMode=BMM_Additive
    scaleSpace=BCS_BoneSpace
    alphaBoolBlend {blendOption=Linear}
    boneToModify {boneName=wolf_calf_r}
    rotation {pitch=-5, yaw=-10}
    scale=<0.2 0.2 0.2>
   }
   Output Pose(Result=Component To Local(ComponentPose=@n9)) {layerGroup=None}
  ]}
 EventGraph {nodes [
   Event BlueprintUpdateAnimation()
   > n1 = Sequence()
   n2 = Cast To FortPlayerPawn(execute=@n1.then_1, Object=GetOwningActor())
   > Set As Fort Player Pawn(As Fort Player Pawn:Fort Player Pawn Object Reference=@n2) -> (Output_Get:Fort Player Pawn Object Reference)
   > n3 = Get!(execute:Exec) -> (then:Exec, else:Exec, As Fort Player Pawn:Fort Player Pawn Object Reference)
   > Set is_wearing_jumpboots(is_wearing_jumpboots:bool=HasMatchingGameplayTag!(self:Gameplay Tag Asset Interface Interface=@n3, TagToCheck:Gameplay Tag Structure=(TagName="Item.Gadget.JumpBoots.AnimTracker")) -> (ReturnValue:bool)) -> (Output_Get:bool)
   n4 = Get AnimInst() -> (AnimInst:Anim Instance Object Reference)
   Is Valid(exec=@n1.then_0, InputObject=@n4)
   > n5 = Branch(Condition="float > float"(A=GetCurveValue(self=@n4, CurveName=FrontEnd_Idle_Bypass), B=0) {is=title K2Node_CallFunction})
   > Set FrontEnd_Idle_Float(FrontEnd_Idle_Float:double=1) -> (Output_Get:double)
   Set FrontEnd_Idle_Float(execute=@n5.else, FrontEnd_Idle_Float=0)
   Event BlueprintInitializeAnimation()
   n6 = GetOwnerSkeletalMeshComponent()
   > Is Valid(InputObject=@n6)
   > Set Skeletal_Mesh(Skeletal_Mesh:Skeletal Mesh Component Object Reference=@n6) -> (Output_Get:Skeletal Mesh Component Object Reference)
   > n7 = Cast To FortPlayerPawn(Object=GetOwningActor())
   > Is Valid(InputObject=@n7)
   > n8 = Set Pawn(Pawn:Fort Player Pawn Object Reference=@n7) -> (Output_Get:Fort Player Pawn Object Reference)
   n9 = GetAnimInstance(self=Get Mesh(self=@n8))
   > Is Valid(InputObject=@n9)
   > Set AnimInst(AnimInst:Anim Instance Object Reference=@n9) -> (Output_Get:Anim Instance Object Reference)
  ]}
]
variables [
 AnimInst {type=AnimInstance, default=null}
 As Fort Player Pawn {type=FortPlayerPawn}
 FrontEnd_Idle_Float {type=number}
 is_wearing_jumpboots {type=boolean}
 Pawn {type=FortPlayerPawn}
 Skeletal_Mesh {type=SkeletalMeshComponent}
]
