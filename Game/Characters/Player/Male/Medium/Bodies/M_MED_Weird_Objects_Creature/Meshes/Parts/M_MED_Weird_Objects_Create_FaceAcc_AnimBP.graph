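# Animation Blueprint for the Weird Objects Creature's face accessory, animating its jaw and flap bones.
# Each state copies the attached parent's pose and curves, then layers a facial sequence over those bones.
# The state machine moves from base to alternate idle during emotes, front-end full-body slot animations, or skydiving with parachute closed.
# When those conditions end, an outro sequence plays and automatically returns to base idle.
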AnimBlueprint /Game/Characters/Player/Male/Medium/Bodies/M_MED_Weird_Objects_Creature/Meshes/Parts/M_MED_Weird_Objects_Create_FaceAcc_AnimBP
target_skeleton=&/M_MED_Weird_Objects_Create_FaceAcc_Skeleton
graphs [
 AnimGraph {nodes [
   AnimGraphNode_StateMachine_0 = Idle States() {is=title AnimGraphNode_StateMachine}
   Output Pose(Result=@AnimGraphNode_StateMachine_0) {layerGroup=None}
  ]}
 Idle States {
  entry=BaseIdle
  path=AnimGraph.AnimGraphNode_StateMachine_0.Idle States
  ids=id
  nodes [
   AnimStateNode_0 = BaseIdle() {is=title AnimStateNode}
   AnimStateNode_1 = EmoteIdle() {is=title AnimStateNode}
   AnimStateNode_2 = EmoteIdle_To_BaseIdle() {is=title AnimStateNode}
   AnimStateTransitionNode_0 = AnimStateTransitionNode() {crossfadeDuration=0.35}
   AnimStateTransitionNode_2 = AnimStateTransitionNode() {crossfadeDuration=0.25}
   AnimStateTransitionNode_3 = AnimStateTransitionNode() {bAutomaticRuleBasedOnSequencePlayerInState=true, crossfadeDuration=0.25}
  ]
  links [
   {
    dst=EmoteIdle
    rule=AnimGraph.AnimGraphNode_StateMachine_0.Idle States.AnimStateTransitionNode_0.Transition
    src=BaseIdle
    via=AnimStateTransitionNode_0
   }
   {
    dst=EmoteIdle_To_BaseIdle
    rule=AnimGraph.AnimGraphNode_StateMachine_0.Idle States.AnimStateTransitionNode_2.Transition
    src=EmoteIdle
    via=AnimStateTransitionNode_2
   }
   {
    dst=BaseIdle
    rule=AnimGraph.AnimGraphNode_StateMachine_0.Idle States.AnimStateTransitionNode_3.Transition
    src=EmoteIdle_To_BaseIdle
    via=AnimStateTransitionNode_3
   }
  ]
 }
 BaseIdle {
  path=AnimGraph.AnimGraphNode_StateMachine_0.Idle States.AnimStateNode_0.BaseIdle
  nodes [
   n1 = Layered blend per bone(
    BasePose=Copy Pose From Mesh() {bCopyCurves=true, bUseAttachedParent=true}
    BlendPoses_0=Sequence Player 'WO_Creature_Facial'() {sequence=/Game/Animation/Game/MainPlayer/Facial/WierdObjects/WO_Creature_Facial}
   ) {layerSetup [{branchFilters [
       {boneName=flap_base_top}
       {boneName=jaw_top}
       {boneName=jaw_bot}
       {boneName=flap_base_top_r}
       {boneName=flap_base_bot_r}
       {boneName=flap_base_bot_l}
       {boneName=flap_base_top_l}
      ]}]}
   Output Animation Pose(Result=@n1)
  ]
 }
 EmoteIdle {
  path=AnimGraph.AnimGraphNode_StateMachine_0.Idle States.AnimStateNode_1.EmoteIdle
  nodes [
   n1 = Layered blend per bone(
    BasePose=Copy Pose From Mesh() {bCopyCurves=true, bUseAttachedParent=true}
    BlendPoses_0=Sequence Player 'WO_Creature_Facial_IdleALT'() {sequence=/Game/Animation/Game/MainPlayer/Facial/WierdObjects/WO_Creature_Facial_IdleALT}
   ) {layerSetup [{branchFilters [
       {boneName=flap_base_top}
       {boneName=jaw_top}
       {boneName=jaw_bot}
       {boneName=flap_base_top_r}
       {boneName=flap_base_bot_r}
       {boneName=flap_base_bot_l}
       {boneName=flap_base_top_l}
      ]}]}
   Output Animation Pose(Result=@n1)
  ]
 }
 EmoteIdle_To_BaseIdle {
  path=AnimGraph.AnimGraphNode_StateMachine_0.Idle States.AnimStateNode_2.EmoteIdle_To_BaseIdle
  nodes [
   n1 = Layered blend per bone(
    BasePose=Copy Pose From Mesh() {bCopyCurves=true, bUseAttachedParent=true}
    BlendPoses_0=Sequence Player 'WO_Creature_Facial_IdleALT_Outro'() {sequence=/Game/Animation/Game/MainPlayer/Facial/WierdObjects/WO_Creature_Facial_IdleALT_Outro}
   ) {layerSetup [{branchFilters [
       {boneName=flap_base_top}
       {boneName=jaw_top}
       {boneName=jaw_bot}
       {boneName=flap_base_top_r}
       {boneName=flap_base_bot_r}
       {boneName=flap_base_bot_l}
       {boneName=flap_base_top_l}
      ]}]}
   Output Animation Pose(Result=@n1)
  ]
 }
 Transition {
  path=AnimGraph.AnimGraphNode_StateMachine_0.Idle States.AnimStateTransitionNode_0.Transition
  nodes [
   n1 = NOT Boolean(A=Get bIsParachuteOpened())
   n2 = Get bIsSkyDiving()
   Result(bCanEnterTransition=OR Boolean!(
     A=Get bIsPlayingEmote()
     B=Get bIsPlayingFullBodySlotInFrontEnd() -> (bIsPlayingFullBodySlotInFrontEnd:bool)
     C:bool=AND Boolean(A=@n2, B=@n1)
    ) -> (ReturnValue))
  ]
 }
 Transition {
  path=AnimGraph.AnimGraphNode_StateMachine_0.Idle States.AnimStateTransitionNode_2.Transition
  nodes [
   n1 = Get bIsPlayingFullBodySlotInFrontEnd()
   n2 = Get bIsPlayingEmote()
   n3 = NOT Boolean(A=Get bIsSkyDiving())
   n4 = Get bIsParachuteOpened()
   Result(bCanEnterTransition=AND Boolean!(A=NOT Boolean(A=@n2), B=NOT Boolean(A=@n1), C:bool=OR Boolean!(A=@n3, B=@n4) -> (ReturnValue)) -> (ReturnValue))
  ]
 }
 Transition {path=AnimGraph.AnimGraphNode_StateMachine_0.Idle States.AnimStateTransitionNode_3.Transition, nodes [Result()]}
 EventGraph {nodes [
   TryGetPawnOwner()
   Event BlueprintUpdateAnimation()
  ]}
]
variables []
